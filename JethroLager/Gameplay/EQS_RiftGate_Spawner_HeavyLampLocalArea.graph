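# EQS query that picks a Rift Gate spawner location in the local area around the querier, apparently for HeavyLamp.
# Two grid options share the same tests; the fallback projects from much lower above the querier.
# Tests screen for water, destructibles, nearby buildings, valid surface and Phoebe-navmesh pathability.
# Scoring weighs safe-zone containment and surface type, and favors points closer to the querier.
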EnvQuery /JethroLager/Gameplay/EQS_RiftGate_Spawner_HeavyLampLocalArea
options [
 {
  generator {
   name=FortEnvQueryGenerator_SimpleGrid_1
   class=FortEnvQueryGenerator_SimpleGrid
   properties {
    bAutoSortTests=true
    bCanRunAsync=false
    envQueryResultNormalizationOption=Default
    generateAround=&EnvQueryContext_Querier
    navDataOverrideContext=null
    optionName=""
    overridenAgentNameForNavmesh=None
    randomDisplacementRatio=0
    gridSize {
     dataBinding=&self:FortEnvQueryGenerator_SimpleGrid_1.FortAIDataProvider_ScalableFloat_0
     dataField=FloatValue
     defaultValue=500
    }
    projectionData {
     bCanDisableTrace=true
     bCanProjectDown=true
     bCanTraceOnGeometry=true
     bCanTraceOnNavMesh=true
     bOnlyBlockingHits=true
     bTraceComplex=false
     extentX=0
     extentY=10
     extentZ=10
     navigationFilter=null
     postProjectionVerticalOffset=0
     projectDown=4000
     projectUp=4000
     serializedChannel=ECC_GameTraceChannel6
     traceChannel=TraceTypeQuery7
     traceMode=GeometryByChannel
     traceProfileName=None
     traceShape=Line
     versionNum=1
    }
    spaceBetween {
     dataBinding=&self:FortEnvQueryGenerator_SimpleGrid_1.FortAIDataProvider_ScalableFloat_1
     dataField=FloatValue
     defaultValue=100
    }
   }
  }
  tests [
   FortQueryTest_InsideAthenaSafeZone_1 {class=FortQueryTest_InsideAthenaSafeZone, properties {testPurpose=Score}}
   EnvQueryTest_Distance_1 {
    class=EnvQueryTest_Distance
    properties {
     distanceTo=&EnvQueryContext_Querier
     filterType=Range
     floatValueMin=0
     scoringEquation=Linear
     scoringFactor=-2
     testComment=""
     testMode=Distance2D
     testPurpose=Score
    }
   }
   EnvQueryTest_Overlap_0 {
    class=EnvQueryTest_Overlap
    properties {testComment=Reject Water, overlapData {extentX=300, extentY=300, extentZ=300, overlapChannel=ECC_GameTraceChannel15, shapeOffset=<0 0 200>}}
   }
   FortQueryTest_ValidSurface_2 {
    class=FortQueryTest_ValidSurface
    properties {
     flatSurfaceToleranceZ {
      dataBinding=&self:FortQueryTest_ValidSurface_2.FortAIDataProvider_ScalableFloat_3
      dataField=FloatValue
      defaultValue=10
     }
     invalidHitActorClasses [
      /Script/FortniteGame.BuildingDeco
      &BuildingProp
      /Script/FortniteGame.BuildingGameplayActor
      &BuildingStairs
      &BuildingWall
      /Script/FortniteGame.BuildingActor
      /Script/Landscape.LandscapeSplineMeshesActor
     ]
     radius {
      dataBinding=&self:FortQueryTest_ValidSurface_2.FortAIDataProvider_ScalableFloat_0
      dataField=FloatValue
      defaultValue=250
     }
     surfaceMaterials [/Game/FortressPhysicalMaterials/Metal, /Game/Balance/Physics/Materials/PhysData_Ice]
     traceOffsetDown {
      dataBinding=&self:FortQueryTest_ValidSurface_2.FortAIDataProvider_ScalableFloat_2
      dataField=FloatValue
      defaultValue=100
     }
     traceOffsetUp {
      dataBinding=&self:FortQueryTest_ValidSurface_2.FortAIDataProvider_ScalableFloat_1
      dataField=FloatValue
      defaultValue=100
     }
     validHitActorClasses [
      /Script/Landscape.LandscapeStreamingProxy
      /Script/Landscape
      &StaticMeshActor
      /Script/Engine.SplineMeshActor
      /Script/Landscape.ControlPointMeshActor
     ]
    }
   }
   FortQueryTest_ValidSurface_3 {
    class=FortQueryTest_ValidSurface
    properties {
     boolValue=false
     scoringFactor=0.1
     testPurpose=Score
     flatSurfaceToleranceZ {
      dataBinding=&self:FortQueryTest_ValidSurface_3.FortAIDataProvider_ScalableFloat_3
      dataField=FloatValue
      defaultValue=10
     }
     radius {
      dataBinding=&self:FortQueryTest_ValidSurface_3.FortAIDataProvider_ScalableFloat_0
      dataField=FloatValue
      defaultValue=250
     }
     surfaceMaterials []
     traceOffsetDown {
      dataBinding=&self:FortQueryTest_ValidSurface_3.FortAIDataProvider_ScalableFloat_2
      dataField=FloatValue
      defaultValue=100
     }
     traceOffsetUp {
      dataBinding=&self:FortQueryTest_ValidSurface_3.FortAIDataProvider_ScalableFloat_1
      dataField=FloatValue
      defaultValue=100
     }
    }
   }
   FortQueryTest_ValidSurface_4 {
    class=FortQueryTest_ValidSurface
    properties {
     scoringFactor=0.2
     testPurpose=Score
     flatSurfaceToleranceZ {
      dataBinding=&self:FortQueryTest_ValidSurface_4.FortAIDataProvider_ScalableFloat_3
      dataField=FloatValue
      defaultValue=10
     }
     radius {
      dataBinding=&self:FortQueryTest_ValidSurface_4.FortAIDataProvider_ScalableFloat_0
      dataField=FloatValue
      defaultValue=250
     }
     surfaceMaterials [/Game/FortressPhysicalMaterials/Asphalt]
     traceOffsetDown {
      dataBinding=&self:FortQueryTest_ValidSurface_4.FortAIDataProvider_ScalableFloat_2
      dataField=FloatValue
      defaultValue=100
     }
     traceOffsetUp {
      dataBinding=&self:FortQueryTest_ValidSurface_4.FortAIDataProvider_ScalableFloat_1
      dataField=FloatValue
      defaultValue=100
     }
    }
   }
   EnvQueryTest_Overlap_1 {
    class=EnvQueryTest_Overlap
    properties {
     testComment=Reject Destructibels
     overlapData {extentX=325, extentY=325, extentZ=325, overlapChannel=ECC_GameTraceChannel11, shapeOffset=<0 0 340>}
    }
   }
   FortQueryTest_HasNearbyBuildings_1 {class=FortQueryTest_HasNearbyBuildings, properties {extentXY=1, testPurpose=Filter}}
   EnvQueryTest_Pathfinding_1 {class=EnvQueryTest_Pathfinding, properties {filterClass=/Game/Athena/AI/Phoebe/NavFilter/NavFilter_Phoebe.NavFilter_Phoebe_C}}
  ]
 }
 {
  generator {
   name=FortEnvQueryGenerator_SimpleGrid_0
   class=FortEnvQueryGenerator_SimpleGrid
   properties {
    bAutoSortTests=true
    bCanRunAsync=false
    envQueryResultNormalizationOption=Default
    generateAround=&EnvQueryContext_Querier
    navDataOverrideContext=null
    optionName=""
    overridenAgentNameForNavmesh=None
    randomDisplacementRatio=0
    gridSize {
     dataBinding=&self:FortEnvQueryGenerator_SimpleGrid_0.FortAIDataProvider_ScalableFloat_0
     dataField=FloatValue
     defaultValue=500
    }
    projectionData {
     bCanDisableTrace=true
     bCanProjectDown=true
     bCanTraceOnGeometry=true
     bCanTraceOnNavMesh=true
     bOnlyBlockingHits=true
     bTraceComplex=false
     extentX=0
     extentY=10
     extentZ=10
     navigationFilter=null
     postProjectionVerticalOffset=0
     projectDown=4000
     projectUp=250
     serializedChannel=ECC_GameTraceChannel6
     traceChannel=TraceTypeQuery7
     traceMode=GeometryByChannel
     traceProfileName=None
     traceShape=Line
     versionNum=1
    }
    spaceBetween {
     dataBinding=&self:FortEnvQueryGenerator_SimpleGrid_0.FortAIDataProvider_ScalableFloat_1
     dataField=FloatValue
     defaultValue=100
    }
   }
  }
  tests [
   FortQueryTest_InsideAthenaSafeZone_0 {class=FortQueryTest_InsideAthenaSafeZone, properties {testPurpose=Score}}
   EnvQueryTest_Distance_0 {
    class=EnvQueryTest_Distance
    properties {
     distanceTo=&EnvQueryContext_Querier
     filterType=Range
     floatValueMin=0
     scoringEquation=Linear
     scoringFactor=-2
     testComment=""
     testMode=Distance2D
     testPurpose=Score
    }
   }
   EnvQueryTest_Overlap_10 {
    class=EnvQueryTest_Overlap
    properties {testComment=Reject Water, overlapData {extentX=300, extentY=300, extentZ=300, overlapChannel=ECC_GameTraceChannel15, shapeOffset=<0 0 200>}}
   }
   FortQueryTest_ValidSurface_5 {
    class=FortQueryTest_ValidSurface
    properties {
     flatSurfaceToleranceZ {
      dataBinding=&self:FortQueryTest_ValidSurface_5.FortAIDataProvider_ScalableFloat_3
      dataField=FloatValue
      defaultValue=10
     }
     invalidHitActorClasses [
      &BuildingDeco
      &BuildingProp
      &BuildingGameplayActor
      &BuildingStairs
      &BuildingWall
      &BuildingActor
      &LandscapeSplineMeshesActor
     ]
     radius {
      dataBinding=&self:FortQueryTest_ValidSurface_5.FortAIDataProvider_ScalableFloat_0
      dataField=FloatValue
      defaultValue=250
     }
     surfaceMaterials [&Metal, &PhysData_Ice]
     traceOffsetDown {
      dataBinding=&self:FortQueryTest_ValidSurface_5.FortAIDataProvider_ScalableFloat_2
      dataField=FloatValue
      defaultValue=100
     }
     traceOffsetUp {
      dataBinding=&self:FortQueryTest_ValidSurface_5.FortAIDataProvider_ScalableFloat_1
      dataField=FloatValue
      defaultValue=100
     }
     validHitActorClasses [
      &LandscapeStreamingProxy
      &Landscape
      &StaticMeshActor
      &SplineMeshActor
      &ControlPointMeshActor
     ]
    }
   }
   FortQueryTest_ValidSurface_0 {
    class=FortQueryTest_ValidSurface
    properties {
     boolValue=false
     scoringFactor=0.1
     testPurpose=Score
     flatSurfaceToleranceZ {
      dataBinding=&self:FortQueryTest_ValidSurface_0.FortAIDataProvider_ScalableFloat_3
      dataField=FloatValue
      defaultValue=10
     }
     radius {
      dataBinding=&self:FortQueryTest_ValidSurface_0.FortAIDataProvider_ScalableFloat_0
      dataField=FloatValue
      defaultValue=250
     }
     surfaceMaterials []
     traceOffsetDown {
      dataBinding=&self:FortQueryTest_ValidSurface_0.FortAIDataProvider_ScalableFloat_2
      dataField=FloatValue
      defaultValue=100
     }
     traceOffsetUp {
      dataBinding=&self:FortQueryTest_ValidSurface_0.FortAIDataProvider_ScalableFloat_1
      dataField=FloatValue
      defaultValue=100
     }
    }
   }
   FortQueryTest_ValidSurface_1 {
    class=FortQueryTest_ValidSurface
    properties {
     scoringFactor=0.2
     testPurpose=Score
     flatSurfaceToleranceZ {
      dataBinding=&self:FortQueryTest_ValidSurface_1.FortAIDataProvider_ScalableFloat_3
      dataField=FloatValue
      defaultValue=10
     }
     radius {
      dataBinding=&self:FortQueryTest_ValidSurface_1.FortAIDataProvider_ScalableFloat_0
      dataField=FloatValue
      defaultValue=250
     }
     surfaceMaterials [&Asphalt]
     traceOffsetDown {
      dataBinding=&self:FortQueryTest_ValidSurface_1.FortAIDataProvider_ScalableFloat_2
      dataField=FloatValue
      defaultValue=100
     }
     traceOffsetUp {
      dataBinding=&self:FortQueryTest_ValidSurface_1.FortAIDataProvider_ScalableFloat_1
      dataField=FloatValue
      defaultValue=100
     }
    }
   }
   EnvQueryTest_Overlap_11 {
    class=EnvQueryTest_Overlap
    properties {
     testComment=Reject Destructibels
     overlapData {extentX=325, extentY=325, extentZ=325, overlapChannel=ECC_GameTraceChannel11, shapeOffset=<0 0 340>}
    }
   }
   FortQueryTest_HasNearbyBuildings_0 {class=FortQueryTest_HasNearbyBuildings, properties {extentXY=1, testPurpose=Filter}}
   EnvQueryTest_Pathfinding_0 {class=EnvQueryTest_Pathfinding, properties {filterClass=&NavFilter_Phoebe_C}}
  ]
 }
]
